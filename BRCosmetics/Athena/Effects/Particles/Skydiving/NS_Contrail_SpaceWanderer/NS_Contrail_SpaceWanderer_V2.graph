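# Niagara skydiving contrail for the SpaceWanderer cosmetic, drawing glowing purple-to-blue energy and fire on the character's skeletal mesh.
# Mesh streaks, rotating body sprites and SubUV fire flipbooks follow sampled mesh locations, with spawn groups selecting bone or socket indices.
# Upward fire variants sample location only at spawn, then drift up.
# A clamped dive value from FN_DivingAttributes, with a front-end override, blends color, size, lifetime and material parameters between rest and dive.
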NiagaraSystem /BRCosmetics/Athena/Effects/Particles/Skydiving/NS_Contrail_SpaceWanderer/NS_Contrail_SpaceWanderer_V2
dependencies {
 usedDataInterfaces [&NiagaraDataInterfaceCurve, &NiagaraDataInterfaceVectorCurve]
 usedDynamicInputs []
 usedModules [&SystemState]
}
emitters {
 Blue_Main {
  emitterUpdateScript [
   &EmitterState(Life Cycle Mode=::NewEnumerator1, Loop Duration=1, UseLoopDelay=false)
   &SpawnRate(SpawnRate=6, Use Spawn Probability=false)
   &SpawnRate(SpawnGroup=1, SpawnRate=6, Use Spawn Probability=false) {module=SpawnRate001}
  ]
  particleSpawnScript [
   /Game/Effects/Niagara/Modules/Skydiving/FN_DivingAttributes/FN_DivingAttributes_v2(Accumulate Dive=0.1, Accumulate Rest=-0.1, SkeletalMesh:NiagaraDataInterfaceSkeletalMesh=$User.SkeletalMesh)
   /SetVariables(Particles.DiveCLAMP:NiagaraFloat=&ClampFloat(Float=&FN_FrontEndFloatModifier(Gameplay=$Particles.Dive, Front End=&Multiply_Float(A=$Particles.Dive, B=10)))) {isSetParameters=true, module=SetVariables_#1}
   /SetVariables(Particles.MaterialRandom:NiagaraFloat=&UniformRangedFloat(Minimum=0, Maximum=1)) {isSetParameters=true, module=SetVariables_#2}
   /Niagara/Modules/Spawn/Initialization/InitializeParticle(
    Color=&LerpLinearColors(
     StartColor=<rgba 28.3334 0 50 0.25>
     EndColor=&LerpLinearColors(
      StartColor=<rgba 0.5667 0 1 0.4>
      EndColor=<rgba 0 40.4868 50 0.4>
      LerpFactor=&UniformAOrBFloat(A=0.5, B=0.7)
     )
     LerpFactor=$Particles.DiveCLAMP
    )
    Lifetime=&Lerp_Float(
     A=&UniformRangedFloat(Minimum=0.5, Maximum=0.7)
     B=&UniformRangedFloat(Minimum=0.8, Maximum=1)
     Alpha=$Particles.DiveCLAMP
    )
    Material Random:NiagaraFloat=$Particles.MaterialRandom
    Mesh Scale=&Multiply_Vector(A=<0.4 0.125 0.125>, B=&Lerp_Vector(A=<1 1 1>, B=<2 1 1>, Alpha=$Particles.DiveCLAMP))
    Module.Write Color=true
    Module.Write Lifetime=true
    Module.Write Mass=true
    Module.Write Material Random=true
    Module.Write Position=false
    Module.Write Scale=true
    Module.Write SpriteRotation=true
    Module.Write SpriteSize=false
    Sprite Rotation=&UniformRangedFloat(Minimum=-0.2, Maximum=0.2)
    Sprite Size=<5 5>
   )
   &SkeletalMeshLocation(
    Apply To Particle Position=false
    Bone / Socket Index=&SpawnGroup()
    Bone Sampling Mode=::NewEnumerator4
    Inherit Velocity=false
    MaskSpawnGroup=false
    Sampled Position Offset=<0 5 0>
    Skeletal Mesh=$User.SkeletalMesh
   ) {module=SkeletalMeshLocation001}
   /Niagara/Modules/Spawn/Orientation/V2/InitialMeshOrientation(
    Orientation Vector=<0 0 -1>
    Rotation=&UniformRangedVector(Minimum=<0 0 0>, Maximum=<1 0 0>)
    Use Orientation Vector=true
    Use Rotation Vector=true
   )
  ]
  particleUpdateScript [
   /SetVariables(Particles.DiveCLAMP:NiagaraFloat=&ClampFloat(Float=&FN_FrontEndFloatModifier(Gameplay=$Particles.Dive, Front End=&Multiply_Float(A=$Particles.Dive, B=10)))) {isSetParameters=true, module=SetVariables_#3}
   &ParticleState()
   &SkeletalMeshLocation(
    Bone / Socket Index=&SpawnGroup()
    Bone Sampling Mode=::NewEnumerator4
    Inherit Velocity=false
    MaskSpawnGroup=false
    Skeletal Mesh=$User.SkeletalMesh
   )
   /Niagara/Modules/Update/Orientation/V2/UpdateMeshOrientation(
    Delta Time:NiagaraFloat=&Multiply_Float(A=$Engine.DeltaTime, B=&Multiply_Float(A=5, B=$Particles.NormalizedAge))
    Rotation Rate:NiagaraFloat=$Particles.Initial.SpriteRotation
    Rotation Vector:Vector3f=<0 0 1>
   ) {enabled=false}
   &ScaleMeshSize(Scale Factor=&Multiply_Vector(
     A=&Lerp_Vector(
      A=&MakeVector(
       X=&FloatFromCurve(FloatCurve=DataInterface() {propertyValues {Curve {keys [<at 0 0.1>,<at 1 2>]}}})
       Y=&FloatFromCurve(FloatCurve=DataInterface() {propertyValues {Curve {keys [
            <at 0 1 RCIM_Cubic>
            {arriveTangent=-1.1222, interpMode=RCIM_Cubic, leaveTangent=-2.1367, tangentMode=RCTM_User, time=0.9, value=0.5}
            <at 1 0.05>
           ]}}})
       Z=&FloatFromCurve(FloatCurve=DataInterface() {propertyValues {Curve {keys [
            <at 0 1 RCIM_Cubic>
            {arriveTangent=-1.1222, interpMode=RCIM_Cubic, leaveTangent=-2.1367, tangentMode=RCTM_User, time=0.9, value=0.5}
            <at 1 0.05>
           ]}}})
      )
      B=&MakeVector(
       X=&FloatFromCurve(FloatCurve=DataInterface() {propertyValues {Curve {keys [<at 0 1>,<at 1 2>]}}})
       Y=&FloatFromCurve(FloatCurve=DataInterface() {propertyValues {Curve {keys [
            {arriveTangent=-1.5003, interpMode=RCIM_Cubic, leaveTangent=-0.2737, tangentMode=RCTM_User, time=0, value=1}
            {arriveTangent=-2.2772, interpMode=RCIM_Cubic, leaveTangent=-2.2772, tangentMode=RCTM_User, time=1, value=0}
           ]}}})
       Z=&FloatFromCurve(FloatCurve=DataInterface() {propertyValues {Curve {keys [
            {arriveTangent=-1.5003, interpMode=RCIM_Cubic, leaveTangent=-0.2737, tangentMode=RCTM_User, time=0, value=1}
            {arriveTangent=-2.2772, interpMode=RCIM_Cubic, leaveTangent=-2.2772, tangentMode=RCTM_User, time=1, value=0}
           ]}}})
      )
      Alpha=$Particles.DiveCLAMP
     )
     B=&VectorFromFloat(Value=&FloatFromCurve(FloatCurve=DataInterface() {propertyValues {Curve {keys [
           <at 0 0 RCIM_Constant>
           {arriveTangent=5.5862, interpMode=RCIM_Cubic, leaveTangent=5.5862, tangentMode=RCTM_User, time=0.3, value=0}
           <at 0.7 1>
           <at 1 1>
          ]}}}))
    )) {module=ScaleMeshSize001}
   &ScaleColor(
    Scale Alpha=&FloatFromCurve(FloatCurve=DataInterface() {propertyValues {Curve {keys [
         {arriveTangent=8.7245, interpMode=RCIM_Cubic, leaveTangent=8.7245, tangentMode=RCTM_User, time=0, value=0}
         <at 0.45 1 RCIM_Cubic>
         {arriveTangent=-0.5281, interpMode=RCIM_Cubic, leaveTangent=-0.5281, tangentMode=RCTM_User, time=1, value=0}
        ]}}})
    ScaleA=true
    ScaleRGB=true
   )
   &DynamicMaterialParameters(
    Index 0 Param 1=&Lerp_Float(A=0, B=0.5, Alpha=$Particles.DiveCLAMP)
    Module.Param0WriteEnabled=true
    Module.Param1WriteEnabled=false
    Module.Param2WriteEnabled=false
    Module.Param3WriteEnabled=false
   )
  ]
  rendererClasses [&NiagaraMeshRendererProperties]
  renderers [{rendererClass=&NiagaraMeshRendererProperties, rendererIndex=0}]
 }
 Blue_Main_Body {
  emitterUpdateScript [
   &EmitterState(Life Cycle Mode=::NewEnumerator1, Loop Duration=1, UseLoopDelay=false)
   &SpawnRate(SpawnRate=10, Use Spawn Probability=false)
   &SpawnRate(SpawnGroup=1, SpawnRate=10, Use Spawn Probability=false) {module=SpawnRate001}
  ]
  particleSpawnScript [
   &FN_DivingAttributes_v2(Accumulate Dive=0.1, Accumulate Rest=-0.1, SkeletalMesh:NiagaraDataInterfaceSkeletalMesh=$User.SkeletalMesh)
   /SetVariables(Particles.DiveCLAMP:NiagaraFloat=&ClampFloat(Float=&FN_FrontEndFloatModifier(Gameplay=$Particles.Dive, Front End=&Multiply_Float(A=$Particles.Dive, B=10)))) {isSetParameters=true, module=SetVariables_#1}
   /SetVariables(Particles.MaterialRandom:NiagaraFloat=&UniformRangedFloat(Minimum=0, Maximum=1)) {isSetParameters=true, module=SetVariables_#2}
   &InitializeParticle(
    Color=&LerpLinearColors(
     StartColor=<rgba 28.3334 0 50 0.025>
     EndColor=&LerpLinearColors(
      StartColor=<rgba 0.3333 0 1 0.025>
      EndColor=<rgba 0 40.4868 50 0.03>
      LerpFactor=&UniformAOrBFloat(A=0.5, B=0.7)
     )
     LerpFactor=$Particles.DiveCLAMP
    )
    Lifetime=&Lerp_Float(
     A=&UniformRangedFloat(Minimum=0.5, Maximum=0.7)
     B=&UniformRangedFloat(Minimum=0.8, Maximum=1)
     Alpha=$Particles.DiveCLAMP
    )
    Material Random:NiagaraFloat=$Particles.MaterialRandom
    Mesh Scale=&Multiply_Vector(A=<0.755 0.32 0.32>, B=&Lerp_Vector(A=<1 1 1>, B=<2 1 1>, Alpha=$Particles.DiveCLAMP))
    Module.Write Color=true
    Module.Write Lifetime=true
    Module.Write Mass=true
    Module.Write Material Random=true
    Module.Write Position=false
    Module.Write Scale=true
    Module.Write SpriteRotation=true
    Module.Write SpriteSize=false
    Sprite Rotation=&UniformRangedFloat(Minimum=-0.2, Maximum=0.2)
    Sprite Size=<5 5>
   )
   &SkeletalMeshLocation(
    Apply To Particle Position=false
    Bone / Socket Index=6
    Bone Sampling Mode=::NewEnumerator4
    Inherit Velocity=false
    MaskSpawnGroup=false
    Sampled Position Offset=<0 5 0>
    Skeletal Mesh=$User.SkeletalMesh
   ) {enabled=false, module=SkeletalMeshLocation001}
   &InitialMeshOrientation(
    Orientation Vector=<0 0 -1>
    Rotation=&UniformRangedVector(Minimum=<0 0 0>, Maximum=<1 0 0>)
    Use Orientation Vector=true
    Use Rotation Vector=true
   )
  ]
  particleUpdateScript [
   /SetVariables(Particles.DiveCLAMP:NiagaraFloat=&ClampFloat(Float=&FN_FrontEndFloatModifier(Gameplay=$Particles.Dive, Front End=&Multiply_Float(A=$Particles.Dive, B=10)))) {isSetParameters=true, module=SetVariables_#3}
   &ParticleState()
   &SkeletalMeshLocation(
    Bone / Socket Index=6
    Bone Sampling Mode=::NewEnumerator4
    Inherit Velocity=false
    MaskSpawnGroup=false
    Sampled Position Offset=<0 8 0>
    Sampled Position Offset Coordinate Space:ENiagaraMeshTransforms=::NewEnumerator2
    Skeletal Mesh=$User.SkeletalMesh
   )
   &UpdateMeshOrientation(
    Delta Time:NiagaraFloat=&Multiply_Float(A=$Engine.DeltaTime, B=&Multiply_Float(A=5, B=$Particles.NormalizedAge))
    Rotation Rate:NiagaraFloat=$Particles.Initial.SpriteRotation
    Rotation Vector:Vector3f=<0 0 1>
   ) {enabled=false}
   &ScaleMeshSize(Scale Factor=&Multiply_Vector(
     A=&Lerp_Vector(
      A=&MakeVector(
       X=&FloatFromCurve(FloatCurve=DataInterface() {propertyValues {Curve {keys [<at 0 0.1>,<at 1 2>]}}})
       Y=&FloatFromCurve(FloatCurve=DataInterface() {propertyValues {Curve {keys [
            <at 0 1 RCIM_Cubic>
            {arriveTangent=-1.1222, interpMode=RCIM_Cubic, leaveTangent=-2.1367, tangentMode=RCTM_User, time=0.9, value=0.5}
            <at 1 0.05>
           ]}}})
       Z=&FloatFromCurve(FloatCurve=DataInterface() {propertyValues {Curve {keys [
            <at 0 1 RCIM_Cubic>
            {arriveTangent=-1.1222, interpMode=RCIM_Cubic, leaveTangent=-2.1367, tangentMode=RCTM_User, time=0.9, value=0.5}
            <at 1 0.05>
           ]}}})
      )
      B=&MakeVector(
       X=&FloatFromCurve(FloatCurve=DataInterface() {propertyValues {Curve {keys [<at 0 1>,<at 1 2>]}}})
       Y=&FloatFromCurve(FloatCurve=DataInterface() {propertyValues {Curve {keys [
            {arriveTangent=-1.5003, interpMode=RCIM_Cubic, leaveTangent=-0.2737, tangentMode=RCTM_User, time=0, value=1}
            {arriveTangent=-2.2772, interpMode=RCIM_Cubic, leaveTangent=-2.2772, tangentMode=RCTM_User, time=1, value=0}
           ]}}})
       Z=&FloatFromCurve(FloatCurve=DataInterface() {propertyValues {Curve {keys [
            {arriveTangent=-1.5003, interpMode=RCIM_Cubic, leaveTangent=-0.2737, tangentMode=RCTM_User, time=0, value=1}
            {arriveTangent=-2.2772, interpMode=RCIM_Cubic, leaveTangent=-2.2772, tangentMode=RCTM_User, time=1, value=0}
           ]}}})
      )
      Alpha=$Particles.DiveCLAMP
     )
     B=&VectorFromFloat(Value=&FloatFromCurve(FloatCurve=DataInterface() {propertyValues {Curve {keys [
           <at 0 0 RCIM_Constant>
           {arriveTangent=5.5862, interpMode=RCIM_Cubic, leaveTangent=5.5862, tangentMode=RCTM_User, time=0.3, value=0}
           <at 0.7 1>
           <at 1 1>
          ]}}}))
    )) {module=ScaleMeshSize001}
   &ScaleColor(
    Scale Alpha=&FloatFromCurve(FloatCurve=DataInterface() {propertyValues {Curve {keys [
         {arriveTangent=8.7245, interpMode=RCIM_Cubic, leaveTangent=8.7245, tangentMode=RCTM_User, time=0, value=0}
         <at 0.45 1 RCIM_Cubic>
         {arriveTangent=-0.5281, interpMode=RCIM_Cubic, leaveTangent=-0.5281, tangentMode=RCTM_User, time=1, value=0}
        ]}}})
    ScaleA=true
    ScaleRGB=true
   )
   &DynamicMaterialParameters(
    Index 0 Param 1=&Lerp_Float(A=0, B=0.5, Alpha=$Particles.DiveCLAMP)
    Module.Param0WriteEnabled=true
    Module.Param1WriteEnabled=false
    Module.Param2WriteEnabled=false
    Module.Param3WriteEnabled=false
   )
  ]
  rendererClasses [&NiagaraMeshRendererProperties]
  renderers [{rendererClass=&NiagaraMeshRendererProperties, rendererIndex=0}]
 }
 Blue_Main_Body001 {
  bEnabled=false
  emitterUpdateScript [
   &EmitterState(Life Cycle Mode=::NewEnumerator1, Loop Duration=1, UseLoopDelay=false)
   &SpawnRate(SpawnRate=3, Use Spawn Probability=false)
  ]
  particleSpawnScript [
   &FN_DivingAttributes_v2(Accumulate Dive=0.1, Accumulate Rest=-0.1, SkeletalMesh:NiagaraDataInterfaceSkeletalMesh=$User.SkeletalMesh)
   /SetVariables(Particles.DiveCLAMP:NiagaraFloat=&ClampFloat(Float=&FN_FrontEndFloatModifier(Gameplay=$Particles.Dive, Front End=&Multiply_Float(A=$Particles.Dive, B=10)))) {isSetParameters=true, module=SetVariables_#1}
   /SetVariables(Particles.MaterialRandom:NiagaraFloat=&UniformRangedFloat(Minimum=0, Maximum=1)) {isSetParameters=true, module=SetVariables_#2}
   &InitializeParticle(
    Color=&LerpLinearColors(
     StartColor=<rgba 28.3334 0 50 0>
     EndColor=&LerpLinearColors(
      StartColor=<rgba 28.3334 0 50 0.15>
      EndColor=<rgba 28.3334 0 50 0.15>
      LerpFactor=&UniformAOrBFloat(A=0.5, B=0.7)
     )
     LerpFactor=$Particles.DiveCLAMP
    )
    Lifetime=&Lerp_Float(
     A=&UniformRangedFloat(Minimum=0.5, Maximum=0.7)
     B=&UniformRangedFloat(Minimum=0.8, Maximum=1)
     Alpha=$Particles.DiveCLAMP
    )
    Material Random:NiagaraFloat=$Particles.MaterialRandom
    Mesh Scale=&Multiply_Vector(A=<0.755 0.32 0.32>, B=&Lerp_Vector(A=<1 1 1>, B=<2.1 1.1 1.1>, Alpha=$Particles.DiveCLAMP))
    Module.Write Color=true
    Module.Write Lifetime=true
    Module.Write Mass=true
    Module.Write Material Random=true
    Module.Write Position=false
    Module.Write Scale=true
    Module.Write SpriteRotation=true
    Module.Write SpriteSize=false
    Sprite Rotation=&UniformRangedFloat(Minimum=-0.2, Maximum=0.2)
    Sprite Size=<5 5>
   )
   &SkeletalMeshLocation(
    Apply To Particle Position=false
    Bone / Socket Index=6
    Bone Sampling Mode=::NewEnumerator4
    Inherit Velocity=false
    MaskSpawnGroup=false
    Sampled Position Offset=<0 5 0>
    Skeletal Mesh=$User.SkeletalMesh
   ) {enabled=false, module=SkeletalMeshLocation001}
   &InitialMeshOrientation(
    Orientation Vector=<0 0 -1>
    Rotation=&UniformRangedVector(Minimum=<0 0 0>, Maximum=<1 0 0>)
    Use Orientation Vector=true
    Use Rotation Vector=true
   )
  ]
  particleUpdateScript [
   /SetVariables(Particles.DiveCLAMP:NiagaraFloat=&ClampFloat(Float=&FN_FrontEndFloatModifier(Gameplay=$Particles.Dive, Front End=&Multiply_Float(A=$Particles.Dive, B=10)))) {isSetParameters=true, module=SetVariables_#3}
   &ParticleState()
   &SkeletalMeshLocation(
    Bone / Socket Index=6
    Bone Sampling Mode=::NewEnumerator4
    Inherit Velocity=false
    MaskSpawnGroup=false
    Sampled Position Offset=<0 8 0>
    Sampled Position Offset Coordinate Space:ENiagaraMeshTransforms=::NewEnumerator2
    Skeletal Mesh=$User.SkeletalMesh
   )
   &UpdateMeshOrientation(
    Delta Time:NiagaraFloat=&Multiply_Float(A=$Engine.DeltaTime, B=&Multiply_Float(A=5, B=$Particles.NormalizedAge))
    Rotation Rate:NiagaraFloat=$Particles.Initial.SpriteRotation
    Rotation Vector:Vector3f=<0 0 1>
   ) {enabled=false}
   &ScaleMeshSize(Scale Factor=&Multiply_Vector(
     A=&Lerp_Vector(
      A=&MakeVector(
       X=&FloatFromCurve(FloatCurve=DataInterface() {propertyValues {Curve {keys [<at 0 0.1>,<at 1 2>]}}})
       Y=&FloatFromCurve(FloatCurve=DataInterface() {propertyValues {Curve {keys [
            <at 0 1 RCIM_Cubic>
            {arriveTangent=-1.1222, interpMode=RCIM_Cubic, leaveTangent=-2.1367, tangentMode=RCTM_User, time=0.9, value=0.5}
            <at 1 0.05>
           ]}}})
       Z=&FloatFromCurve(FloatCurve=DataInterface() {propertyValues {Curve {keys [
            <at 0 1 RCIM_Cubic>
            {arriveTangent=-1.1222, interpMode=RCIM_Cubic, leaveTangent=-2.1367, tangentMode=RCTM_User, time=0.9, value=0.5}
            <at 1 0.05>
           ]}}})
      )
      B=&MakeVector(
       X=&FloatFromCurve(FloatCurve=DataInterface() {propertyValues {Curve {keys [<at 0 1>,<at 1 2>]}}})
       Y=&FloatFromCurve(FloatCurve=DataInterface() {propertyValues {Curve {keys [
            {arriveTangent=-1.5003, interpMode=RCIM_Cubic, leaveTangent=-0.2737, tangentMode=RCTM_User, time=0, value=1}
            {arriveTangent=-2.2772, interpMode=RCIM_Cubic, leaveTangent=-2.2772, tangentMode=RCTM_User, time=1, value=0}
           ]}}})
       Z=&FloatFromCurve(FloatCurve=DataInterface() {propertyValues {Curve {keys [
            {arriveTangent=-1.5003, interpMode=RCIM_Cubic, leaveTangent=-0.2737, tangentMode=RCTM_User, time=0, value=1}
            {arriveTangent=-2.2772, interpMode=RCIM_Cubic, leaveTangent=-2.2772, tangentMode=RCTM_User, time=1, value=0}
           ]}}})
      )
      Alpha=$Particles.DiveCLAMP
     )
     B=&VectorFromFloat(Value=&FloatFromCurve(FloatCurve=DataInterface() {propertyValues {Curve {keys [
           <at 0 0 RCIM_Constant>
           {arriveTangent=5.5862, interpMode=RCIM_Cubic, leaveTangent=5.5862, tangentMode=RCTM_User, time=0.3, value=0}
           <at 0.7 1>
           <at 1 1>
          ]}}}))
    )) {module=ScaleMeshSize001}
   &ScaleColor(
    Scale Alpha=&FloatFromCurve(FloatCurve=DataInterface() {propertyValues {Curve {keys [
         {arriveTangent=8.7245, interpMode=RCIM_Cubic, leaveTangent=8.7245, tangentMode=RCTM_User, time=0, value=0}
         <at 0.45 1 RCIM_Cubic>
         {arriveTangent=-0.5281, interpMode=RCIM_Cubic, leaveTangent=-0.5281, tangentMode=RCTM_User, time=1, value=0}
        ]}}})
    ScaleA=true
    ScaleRGB=true
   )
   &DynamicMaterialParameters(
    Index 0 Param 1=&Lerp_Float(A=0, B=0.5, Alpha=$Particles.DiveCLAMP)
    Module.Param0WriteEnabled=true
    Module.Param1WriteEnabled=false
    Module.Param2WriteEnabled=false
    Module.Param3WriteEnabled=false
   )
  ]
  rendererClasses [&NiagaraMeshRendererProperties]
  renderers [{rendererClass=&NiagaraMeshRendererProperties, rendererIndex=0}]
 }
 Body_Decal {
  emitterUpdateScript [
   &EmitterState(Life Cycle Mode=::NewEnumerator1, Loop Duration=1, UseLoopDelay=false)
   &SpawnRate(SpawnRate=100, Use Spawn Probability=false)
  ]
  particleSpawnScript [
   &FN_DivingAttributes_v2(Accumulate Dive=0.1, Accumulate Rest=-0.1, SkeletalMesh:NiagaraDataInterfaceSkeletalMesh=$User.SkeletalMesh)
   /SetVariables(Particles.DiveCLAMP:NiagaraFloat=&ClampFloat(Float=&FN_FrontEndFloatModifier(Gameplay=$Particles.Dive, Front End=&Multiply_Float(A=$Particles.Dive, B=10)))) {isSetParameters=true}
   /SetVariables(Particles.BodyNRG_dive=2, Particles.BodyNRG_rest=2) {isSetParameters=true}
   &InitializeParticle(
    Color=&MakeLinearColorFromVectorAndFloat(
     "Vector (RGB)"=&Lerp_Vector(
      A=&MakeVectorFromLinearColorRGB(Color=<rgba 5.6667 0 10 1>)
      B=&MakeVectorFromLinearColorRGB(Color=<rgba 0 3.6798 10 1>)
      Alpha=$Particles.DiveCLAMP
     )
     "Float (Alpha)"=&UniformRangedFloat(Minimum=0.6, Maximum=1)
    )
    Lifetime=&UniformRangedFloat(Minimum=0.6, Maximum=0.9)
    Module.Write Color=true
    Module.Write Lifetime=true
    Module.Write Mass=false
    Module.Write Material Random=true
    Module.Write Position=false
    Module.Write Scale=false
    Module.Write SpriteRotation=true
    Module.Write SpriteSize=true
    Position=&Add_Vector(A=/Niagara/DynamicInputs/Helpers/SimulationPosition(), B=<0 0 0>)
    Sprite Rotation=&UniformRangedFloat(Minimum=0, Maximum=180)
    Sprite Size=&Vector2DFromFloat(Value=&Lerp_Float(
      A=&UniformRangedFloat(Minimum=30, Maximum=50)
      B=&UniformRangedFloat(Minimum=65, Maximum=100)
      Alpha=$Particles.DiveCLAMP
     ))
   )
   /Niagara/Modules/Spawn/Location/SphereLocation(MaskSpawnGroup=false, Module.Override Randomness=false, Sphere Radius=30, Surface Only=false)
  ]
  particleUpdateScript [
   /SetVariables(Particles.DiveCLAMP:NiagaraFloat=&ClampFloat(Float=&FN_FrontEndFloatModifier(Gameplay=$Particles.Dive, Front End=&Multiply_Float(A=$Particles.Dive, B=10)))) {isSetParameters=true}
   &ParticleState()
   &ScaleColor(
    Scale Alpha=&FloatFromCurve(FloatCurve=DataInterface() {propertyValues {Curve {keys [<at 0 0 RCIM_Cubic>,<at 0.5 1 RCIM_Cubic>,<at 1 0>]}}})
    ScaleA=true
    ScaleRGB=true
   )
   &SkeletalMeshLocation(
    Bone Sampling Mode=::NewEnumerator0
    Inherit Velocity=false
    MaskSpawnGroup=false
    Module.Override Randomness=false
    Sampled Position Offset=<0 0 20>
    Skeletal Mesh=$User.SkeletalMesh
   )
   &DynamicMaterialParameters(
    Index 0 Param 1=&Lerp_Float(A=$Particles.BodyNRG_rest, B=$Particles.BodyNRG_dive, Alpha=$Particles.DiveCLAMP)
    Module.Param0WriteEnabled=true
    Module.Param1WriteEnabled=false
    Module.Param2WriteEnabled=false
    Module.Param3WriteEnabled=false
   ) {module=DynamicMaterialParameters001}
   &SpriteRotationRate(Rotation Rate=180)
   &ScaleSpriteSize(Scale Factor=&Vector2DFromFloat(Value=&FloatFromCurve(FloatCurve=DataInterface() {propertyValues {Curve {keys [<at 0 1 RCIM_Cubic>, {arriveTangent=-1.2435, interpMode=RCIM_Cubic, leaveTangent=-1.2435, tangentMode=RCTM_User, time=1, value=0.3}]}}})))
  ]
  rendererClasses [&NiagaraSpriteRendererProperties]
  renderers [{rendererClass=&NiagaraSpriteRendererProperties}]
 }
 Fire {
  emitterUpdateScript [
   &EmitterState(Life Cycle Mode=::NewEnumerator1, Loop Duration=1, UseLoopDelay=false)
   &SpawnRate(SpawnRate=2, Use Spawn Probability=false)
   &SpawnRate(SpawnGroup=1, SpawnRate=2, Use Spawn Probability=false) {module=SpawnRate001}
  ]
  particleSpawnScript [
   &FN_DivingAttributes_v2(Accumulate Dive=0.1, Accumulate Rest=-0.1, SkeletalMesh:NiagaraDataInterfaceSkeletalMesh=$User.SkeletalMesh)
   /SetVariables(Particles.DiveCLAMP:NiagaraFloat=&ClampFloat(Float=&FN_FrontEndFloatModifier(Gameplay=$Particles.Dive, Front End=&Multiply_Float(A=$Particles.Dive, B=10)))) {isSetParameters=true, module=SetVariables_#1}
   &InitializeParticle(
    Color=&LerpLinearColors(StartColor=<rgba 28.3334 0 50 0.2>, EndColor=<rgba 0 40.4868 50 0.2>, LerpFactor=$Particles.DiveCLAMP)
    Lifetime=&Lerp_Float(
     A=&UniformRangedFloat(Minimum=0.8, Maximum=1)
     B=&UniformRangedFloat(Minimum=0.5, Maximum=0.7)
     Alpha=$Particles.DiveCLAMP
    )
    Module.Write Color=true
    Module.Write Lifetime=true
    Module.Write Mass=false
    Module.Write Material Random=true
    Module.Write Position=false
    Module.Write Scale=false
    Module.Write SpriteRotation=true
    Module.Write SpriteSize=true
    Position=/Niagara/Functions/Localspace/SimulationPosition()
    Sprite Rotation=&UniformRangedFloat(Minimum=-180, Maximum=180)
    Sprite Size=&Vector2DFromFloat(Value=&Lerp_Float(
      A=&UniformRangedFloat(Minimum=35, Maximum=32)
      B=&UniformRangedFloat(Minimum=35, Maximum=40)
      Alpha=$Particles.DiveCLAMP
     ))
   )
   &DynamicMaterialParameters(
    Index 0 Param 2=&UniformRangedFloat(Minimum=0, Maximum=1)
    Index 0 Param 3=1
    Index 0 Param 4=&UniformRangedFloat(Minimum=1, Maximum=2)
    Module.Param0WriteEnabled=false
    Module.Param1WriteEnabled=true
    Module.Param2WriteEnabled=false
    Module.Param3WriteEnabled=true
   ) {module=DynamicMaterialParameters002}
   &AddVelocity(Velocity=&Lerp_Vector(
     A=&UniformRangedVector(Minimum=<-25 0 0>, Maximum=<25 0 0>)
     B=&UniformRangedVector(Minimum=<-50 0 0>, Maximum=<50 0 0>)
     Alpha=$Particles.DiveCLAMP
    )) {enabled=false}
   &CameraOffset(Camera Offset Amount=35)
  ]
  particleUpdateScript [
   /SetVariables(Particles.DiveCLAMP:NiagaraFloat=&ClampFloat(Float=&FN_FrontEndFloatModifier(Gameplay=$Particles.Dive, Front End=&Multiply_Float(A=$Particles.Dive, B=10)))) {isSetParameters=true, module=SetVariables_#3}
   &ParticleState()
   &SphereLocation(
    Hemisphere X=true
    MaskSpawnGroup=false
    Module.Override Randomness=false
    Sphere Origin:NiagaraPosition=/Niagara/DynamicInputs/Helpers/SimulationPosition()
    Sphere Radius=10
    Surface Only=false
   ) {enabled=false}
   &SkeletalMeshLocation(
    Bone / Socket Index=&SpawnGroup()
    Bone Sampling Mode=::NewEnumerator4
    Inherit Velocity=false
    "Inherit Velocity (Scale)"=0
    MaskSpawnGroup=false
    Skeletal Mesh=$User.SkeletalMesh
   )
   &Drag(Use Linear Drag=true, Use Rotational Drag=false)
   &SolveForcesAndVelocity(Clamp Velocity=false, Limit Acceleration=false)
   /Niagara/Modules/Update/SubUV/SubUVAnimation(Number Of Frames=64)
   &ScaleSpriteSize(Scale Factor=&Vector2DFromFloat(Value=&FloatFromCurve(FloatCurve=DataInterface() {propertyValues {Curve {keys [
          {arriveTangent=7.5764, interpMode=RCIM_Cubic, time=0, value=0.02}
          {arriveTangent=1.2063, interpMode=RCIM_Cubic, leaveTangent=1.2063, tangentMode=RCTM_User, time=0.3, value=0.99}
          <at 1 0.5 RCIM_Cubic>
         ]}}})))
   &DynamicMaterialParameters(
    Index 0 Param 1=&FloatFromCurve(FloatCurve=DataInterface() {propertyValues {Curve {keys [{arriveTangent=2.7551, interpMode=RCIM_Cubic, leaveTangent=2.7551, tangentMode=RCTM_User, time=0, value=0}, <at 0.7 1>]}}})
    Index 0 Param 2=1
    Index 0 Param 3=1
    Index 0 Param 4=1
    Module.Param0WriteEnabled=true
    Module.Param1WriteEnabled=true
    Module.Param2WriteEnabled=true
    Module.Param3WriteEnabled=false
   ) {module=DynamicMaterialParameters001}
   &ScaleColor(
    Scale Alpha=&FloatFromCurve(FloatCurve=DataInterface() {propertyValues {Curve {keys [
         {arriveTangent=8.7245, interpMode=RCIM_Cubic, leaveTangent=8.7245, tangentMode=RCTM_User, time=0, value=0}
         <at 0.2 1 RCIM_Cubic>
         {arriveTangent=-0.5281, interpMode=RCIM_Cubic, leaveTangent=-0.5281, tangentMode=RCTM_User, time=1, value=0}
        ]}}})
    ScaleA=true
    ScaleRGB=true
   )
  ]
  rendererClasses [&NiagaraSpriteRendererProperties]
  renderers [{rendererClass=&NiagaraSpriteRendererProperties}]
 }
 Fire_Body {
  bEnabled=false
  emitterUpdateScript [
   &EmitterState(Life Cycle Mode=::NewEnumerator1, Loop Duration=1, UseLoopDelay=false)
   &SpawnRate(SpawnRate=3, Use Spawn Probability=false)
  ]
  particleSpawnScript [
   &FN_DivingAttributes_v2(Accumulate Dive=0.1, Accumulate Rest=-0.1, SkeletalMesh:NiagaraDataInterfaceSkeletalMesh=$User.SkeletalMesh)
   /SetVariables(Particles.DiveCLAMP:NiagaraFloat=&ClampFloat(Float=&FN_FrontEndFloatModifier(Gameplay=$Particles.Dive, Front End=&Multiply_Float(A=$Particles.Dive, B=10)))) {isSetParameters=true, module=SetVariables_#1}
   &InitializeParticle(
    Color=&LerpLinearColors(StartColor=<rgba 28.3334 0 50 0.02>, EndColor=<rgba 0 40.4868 50 0.05>, LerpFactor=$Particles.DiveCLAMP)
    Lifetime=&Lerp_Float(
     A=&UniformRangedFloat(Minimum=0.8, Maximum=1)
     B=&UniformRangedFloat(Minimum=0.5, Maximum=0.7)
     Alpha=$Particles.DiveCLAMP
    )
    Module.Write Color=true
    Module.Write Lifetime=true
    Module.Write Mass=false
    Module.Write Material Random=true
    Module.Write Position=false
    Module.Write Scale=false
    Module.Write SpriteRotation=true
    Module.Write SpriteSize=true
    Position=/Niagara/Functions/Localspace/SimulationPosition()
    Sprite Rotation=&UniformRangedFloat(Minimum=-180, Maximum=180)
    Sprite Size=&Vector2DFromFloat(Value=&Lerp_Float(
      A=&UniformRangedFloat(Minimum=32, Maximum=35)
      B=&UniformRangedFloat(Minimum=35, Maximum=30)
      Alpha=$Particles.DiveCLAMP
     ))
   )
   &DynamicMaterialParameters(
    Index 0 Param 2=&UniformRangedFloat(Minimum=0, Maximum=1)
    Index 0 Param 3=1
    Index 0 Param 4=&UniformRangedFloat(Minimum=1, Maximum=2)
    Module.Param0WriteEnabled=false
    Module.Param1WriteEnabled=true
    Module.Param2WriteEnabled=false
    Module.Param3WriteEnabled=true
   ) {module=DynamicMaterialParameters002}
   &AddVelocity(Velocity=&Lerp_Vector(
     A=&UniformRangedVector(Minimum=<-25 0 0>, Maximum=<25 0 0>)
     B=&UniformRangedVector(Minimum=<-50 0 0>, Maximum=<50 0 0>)
     Alpha=$Particles.DiveCLAMP
    )) {enabled=false}
   &CameraOffset(Camera Offset Amount=35)
  ]
  particleUpdateScript [
   /SetVariables(Particles.DiveCLAMP:NiagaraFloat=&ClampFloat(Float=&FN_FrontEndFloatModifier(Gameplay=$Particles.Dive, Front End=&Multiply_Float(A=$Particles.Dive, B=10)))) {isSetParameters=true, module=SetVariables_#3}
   &ParticleState()
   &SphereLocation(
    Hemisphere X=true
    MaskSpawnGroup=false
    Module.Override Randomness=false
    Sphere Origin:NiagaraPosition=/Niagara/DynamicInputs/Helpers/SimulationPosition()
    Sphere Radius=10
    Surface Only=false
   )
   &SkeletalMeshLocation(
    Bone / Socket Index=6
    Bone Sampling Mode=::NewEnumerator4
    Inherit Velocity=false
    "Inherit Velocity (Scale)"=0
    MaskSpawnGroup=false
    Sampled Position Offset=<0 10 0>
    Sampled Position Offset Coordinate Space:ENiagaraMeshTransforms=::NewEnumerator2
    Skeletal Mesh=$User.SkeletalMesh
   )
   &Drag(Use Linear Drag=true, Use Rotational Drag=false)
   &SolveForcesAndVelocity(Clamp Velocity=false, Limit Acceleration=false)
   &SubUVAnimation(Number Of Frames=64)
   &ScaleSpriteSize(Scale Factor=&Vector2DFromFloat(Value=&FloatFromCurve(FloatCurve=DataInterface() {propertyValues {Curve {keys [
          {arriveTangent=7.5764, interpMode=RCIM_Cubic, time=0, value=0.02}
          {arriveTangent=1.2063, interpMode=RCIM_Cubic, leaveTangent=1.2063, tangentMode=RCTM_User, time=0.3, value=0.99}
          <at 1 0.5 RCIM_Cubic>
         ]}}})))
   &DynamicMaterialParameters(
    Index 0 Param 1=&FloatFromCurve(FloatCurve=DataInterface() {propertyValues {Curve {keys [{arriveTangent=2.7551, interpMode=RCIM_Cubic, leaveTangent=2.7551, tangentMode=RCTM_User, time=0, value=0}, <at 0.7 1>]}}})
    Index 0 Param 2=1
    Index 0 Param 3=1
    Index 0 Param 4=1
    Module.Param0WriteEnabled=true
    Module.Param1WriteEnabled=true
    Module.Param2WriteEnabled=true
    Module.Param3WriteEnabled=false
   ) {module=DynamicMaterialParameters001}
   &ScaleColor(
    Scale Alpha=&FloatFromCurve(FloatCurve=DataInterface() {propertyValues {Curve {keys [
         {arriveTangent=8.7245, interpMode=RCIM_Cubic, leaveTangent=8.7245, tangentMode=RCTM_User, time=0, value=0}
         <at 0.2 1 RCIM_Cubic>
         {arriveTangent=-0.5281, interpMode=RCIM_Cubic, leaveTangent=-0.5281, tangentMode=RCTM_User, time=1, value=0}
        ]}}})
    ScaleA=true
    ScaleRGB=true
   )
   &ScaleSpriteSize(Scale Factor=&Vector2DFromFloat(Value=2.15)) {module=ScaleSpriteSize001}
  ]
  rendererClasses [&NiagaraSpriteRendererProperties]
  renderers [{rendererClass=&NiagaraSpriteRendererProperties}]
 }
 Fire_Body_Upward {
  emitterUpdateScript [
   &EmitterState(Life Cycle Mode=::NewEnumerator1, Loop Duration=1, UseLoopDelay=false)
   &SpawnRate(SpawnRate=3, Use Spawn Probability=false)
  ]
  particleSpawnScript [
   &FN_DivingAttributes_v2(SkeletalMesh:NiagaraDataInterfaceSkeletalMesh=$User.SkeletalMesh)
   /SetVariables(Particles.DiveCLAMP:NiagaraFloat=&ClampFloat(Float=&FN_FrontEndFloatModifier(Gameplay=$Particles.Dive, Front End=&Multiply_Float(A=$Particles.Dive, B=10)))) {isSetParameters=true, module=SetVariables_#1}
   &InitializeParticle(
    Color=&LerpLinearColors(StartColor=<rgba 28.3334 0 50 0.035>, EndColor=<rgba 0 40.4868 50 0.01>, LerpFactor=$Particles.DiveCLAMP)
    Lifetime=&Lerp_Float(
     A=&UniformRangedFloat(Minimum=0.5, Maximum=0.7)
     B=&UniformRangedFloat(Minimum=0.6, Maximum=0.5)
     Alpha=$Particles.DiveCLAMP
    )
    Module.Write Color=true
    Module.Write Lifetime=true
    Module.Write Mass=false
    Module.Write Material Random=true
    Module.Write Position=false
    Module.Write Scale=false
    Module.Write SpriteRotation=true
    Module.Write SpriteSize=true
    Position=/Niagara/Functions/Localspace/SimulationPosition()
    Sprite Rotation=&UniformRangedFloat(Minimum=360, Maximum=-360)
    Sprite Size=&Vector2DFromFloat(Value=&Lerp_Float(
      A=&UniformRangedFloat(Minimum=30, Maximum=35)
      B=&UniformRangedFloat(Minimum=35, Maximum=40)
      Alpha=$Particles.DiveCLAMP
     ))
   )
   &DynamicMaterialParameters(
    Index 0 Param 2=&UniformRangedFloat(Minimum=0, Maximum=1)
    Index 0 Param 3=1
    Index 0 Param 4=&UniformRangedFloat(Minimum=1, Maximum=2)
    Module.Param0WriteEnabled=false
    Module.Param1WriteEnabled=true
    Module.Param2WriteEnabled=false
    Module.Param3WriteEnabled=true
   ) {module=DynamicMaterialParameters002}
   &AddVelocity(Velocity=&Lerp_Vector(
     A=&UniformRangedVector(Minimum=<-25 0 0>, Maximum=<25 0 0>)
     B=&UniformRangedVector(Minimum=<-50 0 0>, Maximum=<50 0 0>)
     Alpha=$Particles.DiveCLAMP
    )) {enabled=false}
   &AddVelocity(Velocity=&UniformRangedVector(Minimum=<0 0 250>, Maximum=<0 0 50>)) {module=AddVelocity001}
   &SkeletalMeshLocation(
    Bone / Socket Index=6
    Bone Sampling Mode=::NewEnumerator4
    Inherit Velocity=false
    "Inherit Velocity (Scale)"=0
    MaskSpawnGroup=false
    Sampled Position Offset=<0 10 0>
    Skeletal Mesh=$User.SkeletalMesh
   )
   &CameraOffset(Camera Offset Amount=35)
  ]
  particleUpdateScript [
   /SetVariables(Particles.DiveCLAMP:NiagaraFloat=&ClampFloat(Float=&FN_FrontEndFloatModifier(Gameplay=$Particles.Dive, Front End=&Multiply_Float(A=$Particles.Dive, B=10)))) {isSetParameters=true, module=SetVariables_#3}
   &ParticleState()
   &Drag(Drag=2, Use Linear Drag=true, Use Rotational Drag=false)
   &SolveForcesAndVelocity(Clamp Velocity=false, Limit Acceleration=false)
   &SubUVAnimation(Number Of Frames=64)
   &ScaleSpriteSize(Scale Factor=&Vector2DFromFloat(Value=&FloatFromCurve(FloatCurve=DataInterface() {propertyValues {Curve {keys [
          {arriveTangent=-0.402, interpMode=RCIM_Cubic, leaveTangent=-0.402, tangentMode=RCTM_User, time=0, value=0.99}
          {arriveTangent=-0.9468, interpMode=RCIM_Cubic, leaveTangent=-0.9468, tangentMode=RCTM_User, time=1, value=0.5}
         ]}}})))
   &DynamicMaterialParameters(
    Index 0 Param 1=&FloatFromCurve(FloatCurve=DataInterface() {propertyValues {Curve {keys [{arriveTangent=2.7551, interpMode=RCIM_Cubic, leaveTangent=2.7551, tangentMode=RCTM_User, time=0, value=0}, <at 0.7 1>]}}})
    Index 0 Param 2=1
    Index 0 Param 3=1
    Index 0 Param 4=1
    Module.Param0WriteEnabled=true
    Module.Param1WriteEnabled=true
    Module.Param2WriteEnabled=true
    Module.Param3WriteEnabled=false
   ) {module=DynamicMaterialParameters001}
   &ScaleColor(
    Scale Alpha=&FloatFromCurve(FloatCurve=DataInterface() {propertyValues {Curve {keys [
         {arriveTangent=8.7245, interpMode=RCIM_Cubic, leaveTangent=8.7245, tangentMode=RCTM_User, time=0, value=0}
         <at 0.2 1 RCIM_Cubic>
         {arriveTangent=-0.5281, interpMode=RCIM_Cubic, leaveTangent=-0.5281, tangentMode=RCTM_User, time=1, value=0}
        ]}}})
    ScaleA=true
    ScaleRGB=true
   )
   &ScaleSpriteSize(Scale Factor=<1.05 1.05>) {enabled=false, module=ScaleSpriteSize001}
  ]
  rendererClasses [&NiagaraSpriteRendererProperties]
  renderers [{rendererClass=&NiagaraSpriteRendererProperties}]
 }
 Fire_Upward {
  emitterUpdateScript [
   &EmitterState(Life Cycle Mode=::NewEnumerator1, Loop Duration=1, UseLoopDelay=false)
   &SpawnRate(SpawnRate=6, Use Spawn Probability=false)
   &SpawnRate(SpawnGroup=1, SpawnRate=6, Use Spawn Probability=false) {module=SpawnRate001}
   &SpawnPerUnit(Module.Use Max Movement Threshold=true, Module.Use Movement Tolerance=false, Spawn Per Unit=400, Use Spawn Probability=false)
   &SpawnPerUnit(Module.Use Max Movement Threshold=true, Module.Use Movement Tolerance=false, Spawn Group=1, Spawn Per Unit=400, Use Spawn Probability=false) {module=SpawnPerUnit001}
  ]
  particleSpawnScript [
   &FN_DivingAttributes_v2(SkeletalMesh:NiagaraDataInterfaceSkeletalMesh=$User.SkeletalMesh)
   /SetVariables(Particles.DiveCLAMP:NiagaraFloat=&ClampFloat(Float=&FN_FrontEndFloatModifier(Gameplay=$Particles.Dive, Front End=&Multiply_Float(A=$Particles.Dive, B=10)))) {isSetParameters=true, module=SetVariables_#1}
   &InitializeParticle(
    Color=&LerpLinearColors(StartColor=<rgba 28.3334 0 50 0.15>, EndColor=<rgba 0 40.4868 50 0.025>, LerpFactor=$Particles.DiveCLAMP)
    Lifetime=&Lerp_Float(
     A=&UniformRangedFloat(Minimum=0.5, Maximum=0.3)
     B=&UniformRangedFloat(Minimum=0.35, Maximum=0.25)
     Alpha=$Particles.DiveCLAMP
    )
    Module.Write Color=true
    Module.Write Lifetime=true
    Module.Write Mass=false
    Module.Write Material Random=true
    Module.Write Position=false
    Module.Write Scale=false
    Module.Write SpriteRotation=true
    Module.Write SpriteSize=true
    Position=&SimulationPosition()
    Sprite Rotation=&UniformRangedFloat(Minimum=-360, Maximum=360)
    Sprite Size=&Vector2DFromFloat(Value=&Lerp_Float(
      A=&UniformRangedFloat(Minimum=25, Maximum=35)
      B=&UniformRangedFloat(Minimum=25, Maximum=22)
      Alpha=$Particles.DiveCLAMP
     ))
   )
   &DynamicMaterialParameters(
    Index 0 Param 2=&UniformRangedFloat(Minimum=0, Maximum=1)
    Index 0 Param 3=1
    Index 0 Param 4=&UniformRangedFloat(Minimum=1, Maximum=2)
    Module.Param0WriteEnabled=false
    Module.Param1WriteEnabled=true
    Module.Param2WriteEnabled=false
    Module.Param3WriteEnabled=true
   ) {module=DynamicMaterialParameters002}
   &SphereLocation(
    Hemisphere X=true
    MaskSpawnGroup=false
    Module.Override Randomness=false
    Sphere Origin:NiagaraPosition=/Niagara/DynamicInputs/Helpers/SimulationPosition()
    Sphere Radius=10
    Surface Only=false
   ) {enabled=false}
   &SkeletalMeshLocation(
    Bone / Socket Index=&SpawnGroup()
    Bone Sampling Mode=::NewEnumerator4
    Inherit Velocity=false
    "Inherit Velocity (Scale)"=0
    MaskSpawnGroup=false
    Sampled Position Offset Coordinate Space:ENiagaraMeshTransforms=::NewEnumerator1
    Skeletal Mesh=$User.SkeletalMesh
   )
   &AddVelocity(Velocity=&Lerp_Vector(
     A=&UniformRangedVector(Minimum=<-25 0 0>, Maximum=<25 0 0>)
     B=&UniformRangedVector(Minimum=<-50 0 0>, Maximum=<50 0 0>)
     Alpha=$Particles.DiveCLAMP
    )) {enabled=false}
   &AddVelocity(
    Coordinate Space=::World
    Scale Added Velocity=&VectorFromFloat(Value=&Lerp_Float(A=1, B=1, Alpha=$Particles.DiveCLAMP))
    Velocity=&UniformRangedVector(Minimum=<0 0 100>, Maximum=<0 0 70>)
   ) {module=AddVelocity001}
   &CameraOffset(Camera Offset Amount=35)
  ]
  particleUpdateScript [
   /SetVariables(Particles.DiveCLAMP:NiagaraFloat=&ClampFloat(Float=&FN_FrontEndFloatModifier(Gameplay=$Particles.Dive, Front End=&Multiply_Float(A=$Particles.Dive, B=10)))) {isSetParameters=true, module=SetVariables_#3}
   &ParticleState()
   &Drag(Drag=&Lerp_Float(A=0, B=6, Alpha=$Particles.DiveCLAMP), Use Linear Drag=true, Use Rotational Drag=false) {enabled=false}
   &AccelerationForce(
    Acceleration=&Multiply_VectorByFloat(Vector=<0 0 200>, Float=&UniformRangedFloat(Minimum=0.5, Maximum=1))
    Coordinate Space=::Local
   )
   &SolveForcesAndVelocity(Clamp Velocity=false, Limit Acceleration=false)
   &SubUVAnimation(Number Of Frames=64)
   &ScaleSpriteSize(Scale Factor=&Vector2DFromFloat(Value=&FloatFromCurve(FloatCurve=DataInterface() {propertyValues {Curve {keys [
          {arriveTangent=-0.5754, interpMode=RCIM_Cubic, leaveTangent=-0.5754, tangentMode=RCTM_User, time=0, value=0.99}
          {arriveTangent=-0.224, interpMode=RCIM_Cubic, leaveTangent=-0.224, tangentMode=RCTM_User, time=1, value=0.5}
         ]}}})))
   &DynamicMaterialParameters(
    Index 0 Param 1=&FloatFromCurve(FloatCurve=DataInterface() {propertyValues {Curve {keys [{arriveTangent=2.7551, interpMode=RCIM_Cubic, leaveTangent=2.7551, tangentMode=RCTM_User, time=0, value=0}, <at 0.7 1>]}}})
    Index 0 Param 2=1
    Index 0 Param 3=1
    Index 0 Param 4=1
    Module.Param0WriteEnabled=true
    Module.Param1WriteEnabled=true
    Module.Param2WriteEnabled=true
    Module.Param3WriteEnabled=false
   ) {module=DynamicMaterialParameters001}
   &ScaleColor(
    Scale Alpha=&FloatFromCurve(FloatCurve=DataInterface() {propertyValues {Curve {keys [
         {arriveTangent=8.7245, interpMode=RCIM_Cubic, leaveTangent=8.7245, tangentMode=RCTM_User, time=0, value=0}
         <at 0.2 1 RCIM_Cubic>
         {arriveTangent=-0.5281, interpMode=RCIM_Cubic, leaveTangent=-0.5281, tangentMode=RCTM_User, time=1, value=0}
        ]}}})
    ScaleA=true
    ScaleRGB=true
   )
  ]
  rendererClasses [&NiagaraSpriteRendererProperties]
  renderers [{rendererClass=&NiagaraSpriteRendererProperties}]
 }
 SpeedCone_Main {
  emitterUpdateScript [
   &EmitterState(Life Cycle Mode=::NewEnumerator1, Loop Duration=1, UseLoopDelay=false)
   &SpawnRate(SpawnRate=3, Use Spawn Probability=false)
  ]
  particleSpawnScript [
   &FN_DivingAttributes_v2(Accumulate Dive=0.1, Accumulate Rest=-0.1, SkeletalMesh:NiagaraDataInterfaceSkeletalMesh=$User.SkeletalMesh)
   /SetVariables(Particles.DiveAccumulationMax=40, Particles.DiveAccumulationMin=-5) {isSetParameters=true, module=SetVariables_#4}
   /SetVariables(Particles.DiveCLAMP:NiagaraFloat=&ClampFloat(Float=&FN_FrontEndFloatModifier(Gameplay=$Particles.Dive, Front End=&Multiply_Float(A=$Particles.Dive, B=10)))) {isSetParameters=true, module=SetVariables_#5}
   &InitializeParticle(
    Color=&LerpLinearColors(
     StartColor=<rgba 0.5667 0 1 1.5>
     EndColor=&LerpLinearColors(StartColor=<rgba 0.5667 0 1 1>, EndColor=<rgba 0 0.6167 1 1>, LerpFactor=&UniformAOrBFloat(A=0.5, B=0.7))
     LerpFactor=$Particles.DiveCLAMP
    )
    Mesh Scale=&Multiply_Vector(
     A=<0.5 0.6 0.6>
     B=&Lerp_Vector(
      A=&UniformRangedVector(Minimum=<0.3 0.7 0.5>, Maximum=<0.4 0.8 0.6>)
      B=&UniformRangedVector(Minimum=<1 1.5 1.2>, Maximum=<1 1.5 1.3>)
      Alpha=$Particles.Dive
     )
    )
    Module.Write Color=true
    Module.Write Lifetime=true
    Module.Write Mass=false
    Module.Write Material Random=true
    Module.Write Position=false
    Module.Write Scale=true
    Module.Write SpriteRotation=false
    Module.Write SpriteSize=false
    Sprite Size=<5 5>
   )
   &InitialMeshOrientation(
    Orientation Vector=<0 0 1>
    Rotation=&Lerp_Vector(A=<0 0 -0.03>, B=<0 0 0>, Alpha=$Particles.Dive)
    Use Orientation Vector=true
    Use Rotation Vector=true
   )
  ]
  particleUpdateScript [
   /SetVariables(Particles.DivingAttributesExt:NiagaraFloat=&Lerp_Float(A=-5, B=0.2, Alpha=&ClampFloat(Float=&Lerp_Float(A=-1, B=2, Alpha=$Particles.Dive)))) {isSetParameters=true, module=SetVariables_#6}
   /SetVariables(Particles.Dive++:NiagaraFloat=&ClampFloat(
     Float=&Add_Float(A=$Particles.DiveAccumulation, B=$Particles.DivingAttributesExt)
     Min=$Particles.Initial.DiveAccumulationMin
     Max=$Particles.Initial.DiveAccumulationMax
    )) {isSetParameters=true, module=SetVariables_#7}
   /SetVariables(Particles.DiveAccumulation:NiagaraFloat=Linked() {linkedVariable {name=Particles.Dive++, type=NiagaraFloat}}) {isSetParameters=true, module=SetVariables_#8}
   /SetVariables(Particles.DiveCLAMP:NiagaraFloat=&ClampFloat(Float=&FN_FrontEndFloatModifier(Gameplay=$Particles.Dive, Front End=&Multiply_Float(A=$Particles.Dive, B=10)))) {isSetParameters=true, module=SetVariables_#9}
   &ParticleState()
   &SkeletalMeshLocation(
    Bone / Socket Index=6
    Bone Sampling Mode=::NewEnumerator4
    Inherit Velocity=false
    MaskSpawnGroup=false
    Sampled Position Offset=&Lerp_Vector(A=<0 10 -16>, B=<0 10 -25>, Alpha=$Particles.Dive)
    Sampled Position Offset Coordinate Space:ENiagaraMeshTransforms=::NewEnumerator2
    Skeletal Mesh=$User.SkeletalMesh
   )
   &SolveForcesAndVelocity(Clamp Velocity=false, Limit Acceleration=false)
   &ScaleColor(
    Scale Alpha=&FloatFromCurve(FloatCurve=DataInterface() {propertyValues {Curve {keys [
         <at 0 0 RCIM_Cubic>
         <at 0.5 1 RCIM_Cubic>
         {arriveTangent=0.0656, interpMode=RCIM_Cubic, leaveTangent=0.0656, tangentMode=RCTM_User, time=1, value=0}
        ]}}})
    Scale RGB=&VectorFromFloat(Value=&Add_Float(
      A=1
      B=&Multiply_Float(A=$Particles.DiveAccumulation, B=&Divide_Float(A=1, B=$Particles.Initial.DiveAccumulationMax))
     ))
    ScaleA=true
    ScaleRGB=true
   )
   &ScaleMeshSize(Scale Factor=&VectorFromFloat(Value=0.5)) {module=ScaleMeshSize001}
   &ScaleMeshSize(Scale Factor=&ScaleFloatByCurve(
     InitialValue:Vector3f=<1 1 1>
     ScaleCurve:NiagaraDataInterfaceVectorCurve=DataInterface() {propertyValues {XCurve {keys [<at 0 0.25>,<at 1 3>]}, YCurve {keys [<at 0 1.25>,<at 1 3.5>]}, ZCurve {keys [<at 0 1.25>,<at 1 3.5>]}}}
     CurveIndex:NiagaraFloat=$Particles.NormalizedAge
    ))
  ]
  rendererClasses [&NiagaraMeshRendererProperties]
  renderers [{rendererClass=&NiagaraMeshRendererProperties, rendererIndex=0}]
 }
 SpeedCone_Main_Plus {
  emitterUpdateScript [
   &EmitterState(Life Cycle Mode=::NewEnumerator1, Loop Duration=1, UseLoopDelay=false)
   &SpawnRate(SpawnRate=3, Use Spawn Probability=false)
  ]
  particleSpawnScript [
   &FN_DivingAttributes_v2(Accumulate Dive=0.1, Accumulate Rest=-0.1, SkeletalMesh:NiagaraDataInterfaceSkeletalMesh=$User.SkeletalMesh)
   /SetVariables(Particles.DiveAccumulationMax=40, Particles.DiveAccumulationMin=-5) {isSetParameters=true, module=SetVariables_#4}
   /SetVariables(Particles.DiveCLAMP:NiagaraFloat=&ClampFloat(Float=&FN_FrontEndFloatModifier(Gameplay=$Particles.Dive, Front End=&Multiply_Float(A=$Particles.Dive, B=10)))) {isSetParameters=true, module=SetVariables_#5}
   &InitializeParticle(
    Color=&LerpLinearColors(
     StartColor=<rgba 0.5667 0 1 0>
     EndColor=&LerpLinearColors(StartColor=<rgba 0 3.0833 5 2>, EndColor=<rgba 0 3.0833 5 2>, LerpFactor=&UniformAOrBFloat(A=0.5, B=0.7))
     LerpFactor=$Particles.DiveCLAMP
    )
    Mesh Scale=<0.75 0.35 0.35>
    Module.Write Color=true
    Module.Write Lifetime=true
    Module.Write Mass=false
    Module.Write Material Random=true
    Module.Write Position=false
    Module.Write Scale=true
    Module.Write SpriteRotation=false
    Module.Write SpriteSize=false
    Sprite Size=<5 5>
   )
   &InitialMeshOrientation(Orientation Vector=<0 0 1>, Use Orientation Vector=true, Use Rotation Vector=true)
  ]
  particleUpdateScript [
   /SetVariables(Particles.DivingAttributesExt:NiagaraFloat=&Lerp_Float(A=-5, B=0.2, Alpha=&ClampFloat(Float=&Lerp_Float(A=-1, B=2, Alpha=$Particles.Dive)))) {isSetParameters=true, module=SetVariables_#6}
   /SetVariables(Particles.Dive++:NiagaraFloat=&ClampFloat(
     Float=&Add_Float(A=$Particles.DiveAccumulation, B=$Particles.DivingAttributesExt)
     Min=$Particles.Initial.DiveAccumulationMin
     Max=$Particles.Initial.DiveAccumulationMax
    )) {isSetParameters=true, module=SetVariables_#7}
   /SetVariables(Particles.DiveAccumulation:NiagaraFloat=Linked() {linkedVariable {name=Particles.Dive++, type=NiagaraFloat}}) {isSetParameters=true, module=SetVariables_#8}
   /SetVariables(Particles.DiveCLAMP:NiagaraFloat=&ClampFloat(Float=&FN_FrontEndFloatModifier(Gameplay=$Particles.Dive, Front End=&Multiply_Float(A=$Particles.Dive, B=10)))) {isSetParameters=true, module=SetVariables_#9}
   &ParticleState()
   &SkeletalMeshLocation(
    Bone / Socket Index=6
    Bone Sampling Mode=::NewEnumerator4
    Inherit Velocity=false
    MaskSpawnGroup=false
    Sampled Position Offset=&Lerp_Vector(A=<0 0 -18>, B=<0 0 -25>, Alpha=$Particles.Dive)
    Skeletal Mesh=$User.SkeletalMesh
   )
   &SolveForcesAndVelocity(Clamp Velocity=false, Limit Acceleration=false)
   &ScaleColor(
    Scale Alpha=&FloatFromCurve(FloatCurve=DataInterface() {propertyValues {Curve {keys [
         <at 0 0 RCIM_Cubic>
         <at 0.5 1 RCIM_Cubic>
         {arriveTangent=0.0656, interpMode=RCIM_Cubic, leaveTangent=0.0656, tangentMode=RCTM_User, time=1, value=0}
        ]}}})
    Scale RGB=&VectorFromFloat(Value=&Add_Float(
      A=1
      B=&Multiply_Float(A=$Particles.DiveAccumulation, B=&Divide_Float(A=1, B=$Particles.Initial.DiveAccumulationMax))
     ))
    ScaleA=true
    ScaleRGB=true
   )
   &ScaleMeshSize(Scale Factor=&VectorFromFloat(Value=0.5)) {module=ScaleMeshSize001}
   &ScaleMeshSize(Scale Factor=&ScaleFloatByCurve(
     InitialValue:Vector3f=<1 1 1>
     ScaleCurve:NiagaraDataInterfaceVectorCurve=DataInterface() {propertyValues {XCurve {keys [<at 0 0.25>,<at 1 3>]}, YCurve {keys [<at 0 1.25>,<at 1 3.5>]}, ZCurve {keys [<at 0 1.25>,<at 1 3.5>]}}}
     CurveIndex:NiagaraFloat=$Particles.NormalizedAge
    ))
   &DynamicMaterialParameters(
    Index 0 Param 1=3
    Index 0 Param 2=1
    Module.Param0WriteEnabled=true
    Module.Param1WriteEnabled=true
    Module.Param2WriteEnabled=true
    Module.Param3WriteEnabled=true
   )
  ]
  rendererClasses [&NiagaraMeshRendererProperties]
  renderers [{rendererClass=&NiagaraMeshRendererProperties, rendererIndex=0}]
 }
 crystal_shards_SpeedDebris {
  emitterUpdateScript [
   &EmitterState(Life Cycle Mode=::NewEnumerator1, Loop Duration=1, UseLoopDelay=false)
   &SpawnRate(Spawn Probability=0.6, SpawnRate=30, Use Spawn Probability=true)
  ]
  particleSpawnScript [
   &FN_DivingAttributes_v2(Accumulate Dive=0.1, Accumulate Rest=-0.1, SkeletalMesh:NiagaraDataInterfaceSkeletalMesh=$User.SkeletalMesh)
   /SetVariables(Particles.DiveCLAMP:NiagaraFloat=&ClampFloat(Float=&FN_FrontEndFloatModifier(Gameplay=$Particles.Dive, Front End=&Multiply_Float(A=$Particles.Dive, B=10)))) {isSetParameters=true}
   &InitializeParticle(
    Color=&LerpLinearColors(StartColor=<rgba 28.3334 0 50 1>, EndColor=<rgba 0 40.4868 50 1>, LerpFactor=$Particles.DiveCLAMP)
    Lifetime=&UniformRangedFloat(Minimum=0.3, Maximum=0.9)
    Module.Write Color=true
    Module.Write Lifetime=true
    Module.Write Mass=false
    Module.Write Material Random=true
    Module.Write Position=false
    Module.Write Scale=false
    Module.Write SpriteRotation=false
    Module.Write SpriteSize=true
    Sprite Size=&Vector2DFromFloat(Value=&Multiply_Float(
      A=20
      B=&Lerp_Float(
       A=&UniformRangedFloat(Minimum=0.5, Maximum=1)
       B=&UniformRangedFloat(Minimum=1, Maximum=1.25)
       Alpha=$Particles.DiveCLAMP
      )
     ))
   )
   &AddVelocity(Velocity=&MakeVector(
     X=&Lerp_Float(
      A=&UniformRangedFloat(Minimum=-20, Maximum=20)
      B=&UniformRangedFloat(Minimum=-20, Maximum=20)
      Alpha=$Particles.DiveCLAMP
     )
     Y=0
     Z=0
    ))
   &SkeletalMeshLocation(
    Bone Sampling Mode=::NewEnumerator2
    Inherit Velocity=false
    MaskSpawnGroup=false
    Module.Override Randomness=false
    Skeletal Mesh=DataInterface() {propertyValues {
      DefaultMesh=null
      ExcludeBoneName=root
      OptionalComponentClass=null
      PreviewMesh=&M_Med_Soldier_04_Preview
      SkinningMode=SkinOnTheFly
      SoftSourceActor=null
      SourceMode=Default
      UvSetIndex=0
      WholeMeshLOD=-1
      bExcludeBone=false
      bReadDeformedGeometry=true
      bRequireCurrentFrameData=true
      ComponentTags []
      FilteredBones [root, ik_foot_root, ik_hand_root, VB Root_hand_l, VB Root_hand_r, attach, VB root_prop]
      FilteredSockets [None]
      MeshUserParameter {parameter {name=None, typeDefHandle {registeredTypeIndex=1}, varData []}}
      SamplingRegions []
     }}
   )
   &CylinderLocation(
    Cylinder Height=30
    Cylinder Radius=&UniformRangedFloat(Minimum=5, Maximum=30)
    MaskSpawnGroup=false
    Module.Enable Lathe Profile=false
    Module.Override Randomness=false
    Module.Surface Only=true
    Orientation Axis:ENiagaraOrientationAxis=::XAxis
    Use Endcaps In Surface Only Mode=false
   )
  ]
  particleUpdateScript [
   /SetVariables(Particles.DiveCLAMP:NiagaraFloat=&ClampFloat(Float=&FN_FrontEndFloatModifier(Gameplay=$Particles.Dive, Front End=&Multiply_Float(A=$Particles.Dive, B=10)))) {isSetParameters=true}
   &ParticleState()
   &AddVelocity(Velocity=&MakeVector(
     X=0
     Y=0
     Z=&Lerp_Float(
      A=&UniformRangedFloat(Minimum=30, Maximum=45)
      B=&UniformRangedFloat(Minimum=60, Maximum=90)
      Alpha=Linked() {linkedVariable {name=Particles.IntegratedDiveCLAMP+, type=NiagaraFloat}}
     )
    )) {module=AddVelocity002}
   &SolveForcesAndVelocity(Clamp Velocity=false, Limit Acceleration=false)
   &ScaleColor(
    Scale Alpha=&FloatFromCurve(FloatCurve=DataInterface() {propertyValues {Curve {keys [<at 0 1>,<at 1 0>]}}})
    Scale RGB=&VectorFromFloat(Value=&FloatFromCurve(FloatCurve=DataInterface() {propertyValues {Curve {keys [<at 0 1 RCIM_Cubic>, {arriveTangent=-1.8097, interpMode=RCIM_Cubic, leaveTangent=-1.8097, tangentMode=RCTM_User, time=1, value=0}]}}}))
    ScaleA=true
    ScaleRGB=true
   )
   &ScaleSpriteSize(Scale Factor=&Vector2DFromFloat(Value=&Multiply_Float(A=$Particles.NormalizedAge, B=&OneMinusFloat(Float=$Particles.NormalizedAge))))
   &ScaleSpriteSizeBySpeed(
    Max Scale Factor=<0.5 40>
    Min Scale Factor=<0.1 1>
    Velocity Threshold:NiagaraFloat=&Lerp_Float(A=1800, B=1200, Alpha=Linked() {linkedVariable {name=Particles.IntegratedDiveCLAMP+, type=NiagaraFloat}})
   )
  ]
  rendererClasses [&NiagaraSpriteRendererProperties]
  renderers [{rendererClass=&NiagaraSpriteRendererProperties}]
 }
 crystal_shards_blue {
  emitterUpdateScript [
   &EmitterState(Life Cycle Mode=::NewEnumerator1, Loop Duration=1, UseLoopDelay=false)
   &SpawnRate(SpawnRate=60, Use Spawn Probability=false)
  ]
  particleSpawnScript [
   &FN_DivingAttributes_v2(Accumulate Dive=0.1, Accumulate Rest=-0.1, SkeletalMesh:NiagaraDataInterfaceSkeletalMesh=$User.SkeletalMesh)
   /SetVariables(Particles.DiveCLAMP:NiagaraFloat=&ClampFloat(Float=&FN_FrontEndFloatModifier(Gameplay=$Particles.Dive, Front End=&Multiply_Float(A=$Particles.Dive, B=10)))) {isSetParameters=true}
   &InitializeParticle(
    Color=&LerpLinearColors(StartColor=<rgba 28.3334 0 50 1>, EndColor=<rgba 0 40.4868 50 1>, LerpFactor=$Particles.DiveCLAMP)
    Lifetime=&UniformRangedFloat(Minimum=0.7, Maximum=1.1)
    Mesh Scale=&Multiply_Vector(
     A=<0.1 0.1 0.025>
     B=&VectorFromFloat(Value=&Lerp_Float(
       A=&UniformRangedFloat(Minimum=0.3, Maximum=0.7)
       B=&UniformRangedFloat(Minimum=0.1, Maximum=0.3)
       Alpha=$Particles.DiveCLAMP
      ))
    )
    Module.Write Color=true
    Module.Write Lifetime=true
    Module.Write Mass=false
    Module.Write Material Random=true
    Module.Write Position=true
    Module.Write Scale=true
    Module.Write SpriteRotation=false
    Module.Write SpriteSize=false
    Position=&Add_Vector(A=/Niagara/Functions/Localspace/SimulationPosition(), B=<0 0 0>)
    Sprite Size=<5 5>
   )
   &AddVelocity(Velocity=&MakeVector(
     X=0
     Y=0
     Z=&Lerp_Float(
      A=&UniformRangedFloat(Minimum=0, Maximum=0)
      B=&UniformRangedFloat(Minimum=600, Maximum=1200)
      Alpha=$Particles.DiveCLAMP
     )
    )) {module=AddVelocity001}
   &CylinderLocation(
    Cylinder Height=30
    Cylinder Radius=&UniformRangedFloat(Minimum=5, Maximum=30)
    MaskSpawnGroup=false
    Module.Enable Lathe Profile=false
    Module.Override Randomness=false
    Module.Surface Only=true
    Non Uniform Scale:Vector3f=<1.2 1.2 0.5>
    Orientation Axis:ENiagaraOrientationAxis=::XAxis
    Use Endcaps In Surface Only Mode=false
   )
   &InitialMeshOrientation(
    Orientation Vector=<0 0 1>
    Rotation=&UniformRangedVector(Minimum=<0 0 0>, Maximum=<1 0 0>)
    Use Orientation Vector=true
    Use Rotation Vector=true
   )
   &SkeletalMeshLocation(
    Bone Sampling Mode=::NewEnumerator2
    Inherit Velocity=false
    MaskSpawnGroup=false
    Module.Override Randomness=false
    Skeletal Mesh=DataInterface() {propertyValues {
      DefaultMesh=null
      ExcludeBoneName=root
      OptionalComponentClass=null
      PreviewMesh=&M_Med_Soldier_04_Preview
      SkinningMode=SkinOnTheFly
      SoftSourceActor=null
      SourceMode=Default
      UvSetIndex=0
      WholeMeshLOD=-1
      bExcludeBone=false
      bReadDeformedGeometry=true
      bRequireCurrentFrameData=true
      ComponentTags []
      FilteredBones [root, ik_foot_root, ik_hand_root, VB Root_hand_l, VB Root_hand_r, attach, VB root_prop]
      FilteredSockets [None]
      MeshUserParameter {parameter {name=None, typeDefHandle {registeredTypeIndex=1}, varData []}}
      SamplingRegions []
     }}
   )
  ]
  particleUpdateScript [
   /SetVariables(Particles.DiveCLAMP:NiagaraFloat=&ClampFloat(Float=&FN_FrontEndFloatModifier(Gameplay=$Particles.Dive, Front End=&Multiply_Float(A=$Particles.Dive, B=10)))) {isSetParameters=true}
   &ParticleState()
   &AddVelocity(Velocity=&MakeVector(
     X=0
     Y=0
     Z=&Lerp_Float(
      A=&UniformRangedFloat(Minimum=3, Maximum=15)
      B=&UniformRangedFloat(Minimum=30, Maximum=45)
      Alpha=$Particles.DiveCLAMP
     )
    ))
   /Niagara/Modules/Update/Forces/CurlNoiseForce(Noise Frequency=5, Noise Strength=150, Pan Noise=false, Random Seed=1)
   &SolveForcesAndVelocity(Clamp Velocity=false, Limit Acceleration=false)
   &ScaleMeshSize(Scale Factor=&VectorFromFloat(Value=&FloatFromCurve(FloatCurve=DataInterface() {propertyValues {Curve {keys [
          <at 0 1>
          {arriveTangent=-0.5431, interpMode=RCIM_Cubic, leaveTangent=-0.5431, tangentMode=RCTM_User, time=0.6, value=1}
          {arriveTangent=-5.5854, interpMode=RCIM_Cubic, leaveTangent=-5.5854, tangentMode=RCTM_User, time=1, value=0}
         ]}}})))
   &ScaleColor(
    Scale Alpha=&FloatFromCurve(FloatCurve=DataInterface() {propertyValues {Curve {keys [<at 0 1>,<at 1 0>]}}})
    Scale RGB=&VectorFromFloat(Value=&FloatFromCurve(FloatCurve=DataInterface() {propertyValues {Curve {keys [<at 0 1>,<at 1 0>]}}}))
    ScaleA=true
    ScaleRGB=true
   )
   &ScaleMeshSizeBySpeed(
    Max Scale Factor=<100 0.6 0.6>
    Sample Scale Factor By Curve=true
    Scale Factor Curve=DataInterface() {propertyValues {Curve {keys [<at 0 0>,<at 0.3 0>,<at 1 1>]}}}
    Speed Threshold=900
   )
  ]
  rendererClasses [&NiagaraMeshRendererProperties]
  renderers [{rendererClass=&NiagaraMeshRendererProperties, rendererIndex=0}]
 }
}
summary {
 systemName=NS_Contrail_SpaceWanderer_V2
 emitters [
  {emitterName=Body_Decal, rendererClasses [&NiagaraSpriteRendererProperties]}
  {emitterName=crystal_shards_blue, rendererClasses [&NiagaraMeshRendererProperties]}
  {emitterName=SpeedCone_Main, rendererClasses [&NiagaraMeshRendererProperties]}
  {emitterName=crystal_shards_SpeedDebris, rendererClasses [&NiagaraSpriteRendererProperties]}
  {emitterName=Fire, rendererClasses [&NiagaraSpriteRendererProperties]}
  {emitterName=Blue_Main, rendererClasses [&NiagaraMeshRendererProperties]}
  {emitterName=Blue_Main_Body, rendererClasses [&NiagaraMeshRendererProperties]}
  {emitterName=Fire_Upward, rendererClasses [&NiagaraSpriteRendererProperties]}
  {bEnabled=false, emitterName=Fire_Body, rendererClasses [&NiagaraSpriteRendererProperties]}
  {emitterName=Fire_Body_Upward, rendererClasses [&NiagaraSpriteRendererProperties]}
  {emitterName=SpeedCone_Main_Plus, rendererClasses [&NiagaraMeshRendererProperties]}
  {bEnabled=false, emitterName=Blue_Main_Body001, rendererClasses [&NiagaraMeshRendererProperties]}
 ]
 userVariables [
  User.bIsFrontEnd {type=NiagaraBool}
  User.bIsFrontEndPreview {type=NiagaraBool}
  User.PlayerSpeed {type=NiagaraFloat, defaultValue=0, description=""}
  User.RibbonIDOffset {type=NiagaraInt32}
  User.SizeScale {type=NiagaraFloat, defaultValue=0, description=""}
  User.SkeletalMesh {
   type=NiagaraDataInterfaceSkeletalMesh
   defaultValue {
    struct=NiagaraExt_VariableValue_DataInterface
    value {dataInterface=&self:NiagaraDataInterfaceSkeletalMesh_2}
   }
  }
 ]
}
